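# Translucent unlit skeletal-mesh material for the Washy vehicle wheel cosmetic, showing animated, tinted sloshing water with soap foam and bubbles.
# Sine-driven panners build wavy UVs for clamped gradient samples, forming a water mask that blends panning soap layers with rotating, bump-offset bubbles.
# The tinted result drives emissive, while foam, water mask, vertex red and a mask texture set opacity.
# Output passes through the shared vehicle damage and de-rez transition functions.
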Material /VehicleCosmetics/Wheels/Washy/Materials/M_Wheel_Washy_SE
graph {
 properties {
  bAutomaticallySetUsageInEditor=false
  bUseMaterialAttributes=true
  bUsedWithSkeletalMesh=true
  blendMode=BLEND_Translucent
  shadingModel=MSM_Unlit
  usedShadingModels=Unlit
 }
 nodes [
  n1 = Panner(Coordinate=TextureCoordinate() {uTiling=0.5, vTiling=0.5}) {speedY=-0.1}
  n2 = TextureCoordinate() {uTiling=0.3, vTiling=0.3}
  n3 = Add(
   A=TextureSample(UVs=Panner() {speedY=-0.2}) {texture=/VehicleCosmetics/Wheels/Octo/Textures/Bubbles}.R
   B=TextureSample(UVs=Panner(Coordinate=@n2) {speedY=-0.1}) {texture=&Bubbles}.R
  )
  n4 = ScalarParameter() {defaultValue=2, group=Washy, parameterName=SoamLayer2Intensity}
  n5 = Panner(Coordinate=TextureCoordinate()) {speedX=-1}
  n6 = Time()
  n7 = Constant() {r=-0.8}
  n8 = Sine(None=Sine(None=Add(A=@n6, B=@n7)) {period=6}) {period=3}
  n9 = Constant() {r=0.5}
  n10 = Panner(Time=Multiply(A=Add(A=@n8, B=@n9), B=Constant() {r=0.5})) {speedX=2, speedY=-0.6}
  n11 = Panner(Coordinate=AppendVector(A=ComponentMask(None=@n5) {r=true}, B=ComponentMask(None=@n10) {g=true}), Time=Constant() {r=0.3}) {speedY=0.7}
  n12 = Add(
   A=Multiply(
    A=ScalarParameter() {defaultValue=0.4, group=Washy, parameterName=SoapLayer1Intensity}
    B=TextureSample(UVs=@n1) {texture=&Bubbles}.R
   )
   B=Multiply(
    A=Multiply(A=@n3, B=@n4)
    B=TextureSample(UVs=@n11) {samplerSource=SSM_Clamp_WorldGroupSettings, texture=/VehicleCosmetics/Wheels/Washy/Textures/T_Washy_Gradient_Clamped}.RGB
   )
  )
  n13 = Sine(None=Time()) {period=6}
  n14 = Multiply(A=Add(A=Sine(None=@n13) {period=3}, B=Constant() {r=0.5}), B=Constant() {r=0.5})
  n15 = Panner(Coordinate=TextureCoordinate() {uTiling=0.5}) {speedX=-0.5}
  n16 = Add(A=Sine(None=ComponentMask(None=@n15) {r=true}), B=Constant() {r=1})
  n17 = Constant() {r=0.5}
  n18 = ComponentMask(None=MaterialFunctionCall(In=Multiply(A=@n16, B=@n17), Contrast=Constant() {r=0.1}) {materialFunction=&CheapContrast}.Result) {r=true}
  n19 = Constant() {r=0.2}
  n20 = TextureCoordinate() {uTiling=0.5}
  n21 = Sine(None=ComponentMask(None=Panner(Coordinate=@n20) {speedX=-1}) {r=true})
  n22 = Constant() {r=0.1}
  n23 = Add(A=ComponentMask(None=Panner(Time=@n14) {speedX=2, speedY=-0.6}) {g=true}, B=Add(A=Multiply(A=@n18, B=@n19), B=Multiply(A=@n21, B=@n22)))
  n24 = Time()
  n25 = Multiply(A=Add(A=Sine(None=@n24) {period=1.35}, B=Constant() {r=0.5}), B=Constant() {r=0.5})
  n26 = AppendVector(A=ComponentMask(None=Panner(Time=@n25) {speedX=-0.5}) {r=true}, B=@n23)
  n27 = TextureSample(UVs=Rotator(Coordinate=@n26) {speed=1}) {texture=&Bubbles}
  n28 = AppendVector(A=ComponentMask(None=Panner(Time=@n25) {speedX=-1}) {r=true}, B=@n23)
  n29 = Rotator(Coordinate=@n28) {speed=-0.8}
  n30 = Constant() {r=-2}
  n31 = Multiply(
   A=TextureSample(UVs=BumpOffset(Coordinate=@n29, Height=@n30)) {texture=&Bubbles}.R
   B=Constant() {r=0.5}
  )
  n32 = TextureSample(UVs=@n28) {
   samplerSource=SSM_Clamp_WorldGroupSettings
   samplerType=SAMPLERTYPE_LinearColor
   texture=/VehicleCosmetics/Wheels/Washy/Textures/T_Washy_Gradient
  }
  n33 = TextureSample(UVs=@n26) {
   samplerSource=SSM_Clamp_WorldGroupSettings
   samplerType=SAMPLERTYPE_LinearColor
   texture=&T_Washy_Gradient
  }
  n34 = ComponentMask(None=Panner(Time=@n25) {speedX=0.25}) {r=true}
  n35 = TextureSample(UVs=Panner(Coordinate=AppendVector(A=@n34, B=@n23))) {
   samplerSource=SSM_Clamp_WorldGroupSettings
   samplerType=SAMPLERTYPE_LinearColor
   texture=&T_Washy_Gradient
  }
  n36 = Clamp(None=Add(A=Multiply(A=@n32.A, B=@n33.A), B=@n35.A))
  n37 = Multiply(A=@n32.R, B=@n33.R)
  n38 = Add(A=Multiply(A=Add(A=@n27.RGB, B=@n31), B=@n36), B=Clamp(None=Multiply(A=@n37, B=@n35.R)))
  n39 = Multiply(
   A=ScalarParameter() {defaultValue=2, group=Washy, parameterName=Foam Brightness}
   B=LinearInterpolate(A=@n12, B=Clamp(None=@n38), Alpha=@n36)
  )
  n40 = Multiply(A=@n36, B=ScalarParameter() {defaultValue=0.25, group=Washy, parameterName=Water Brightness})
  n41 = VectorParameter() {group=Washy, parameterName=Color, defaultValue=<rgba 0.0786 0.3177 0.3594 1>}
  n42 = ScalarParameter() {defaultValue=1, group=Washy, parameterName=WaterColorBrightness}
  n43 = Clamp(None=Multiply(A=Add(A=@n39, B=@n40), B=Multiply(A=@n41.RGB, B=@n42)))
  n44 = Multiply(A=@n12, B=Constant() {r=2.5})
  n45 = Constant() {r=1}
  n46 = LinearInterpolate(
   A=Constant()
   B=LinearInterpolate(A=LinearInterpolate(A=@n44, B=@n45, Alpha=@n36), B=Constant() {r=0.7}, Alpha=VertexColor().R)
   Alpha=TextureSample() {texture=/VehicleCosmetics/Wheels/WashyFlare/Textures/T_Washy_SE_Mask}.RGB
  )
  MaterialFunctionCall(In=MaterialFunctionCall(Material=SetMaterialAttributes!(MaterialAttributes, Emissive Color=@n43, Opacity=@n46) {attributeSetTypes [#1, #2]}) {materialFunction=&MF_BaseVehicle_Damage}.Result) {materialFunction=&MF_VehicleDeRez_Transition}
 ]
}
